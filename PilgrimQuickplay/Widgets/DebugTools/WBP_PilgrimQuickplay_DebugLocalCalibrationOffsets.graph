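# Debug widget for Pilgrim Quickplay that displays the locally measured A/V offset and input latency offset, in milliseconds, in two stacked text lines.
# On Construct, if either FMCalibration CVar (ShowInputHistogram or IsCalibrationDebugEnabled) equals 1, it sets itself SelfHitTestInvisible and fills both lines from the measured render-offset getters.
# The values are read only at construction, with no tick or refresh.
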WidgetBlueprint /PilgrimQuickplay/Widgets/DebugTools/WBP_PilgrimQuickplay_DebugLocalCalibrationOffsets
graphs [EventGraph {nodes [
   Event Construct()
   n1 = GetNumericCVar!(CVar:String=FMCalibration.ShowInputHistogram) -> (OutValue:float, ReturnValue:bool)
   n2 = GetNumericCVar(CVar=FMCalibration.IsCalibrationDebugEnabled)
   > Branch(Condition=OR Boolean(A="Equal (Float)"(A=@n1, B=1), B="Equal (Float)"(A=@n2, B=1)))
   > SetVisibility(self=Self-Reference(), InVisibility=SelfHitTestInvisible)
   > n3 = Sequence()
   > n4 = GetMeasuredVideoToAudioRenderOffsetMs!(execute) -> (then, ReturnValue:float)
   > "SetText (Text)"(
    self=Get AVOffset() -> (AVOffset:Common Text Object Reference)
    InText=Format Text!(Format="LOCTEXT(\"A/V Offset: {Val} ms\")", Val:float=@n4) -> (Result)
   )
   n5 = GetMeasuredUserExperienceAndReactionToAudioRenderOffsetMs!(execute=@n3.then_1) -> (then, ReturnValue:float)
   > "SetText (Text)"(
    self=Get InputOffset() -> (InputOffset:Common Text Object Reference)
    InText=Format Text(Format="LOCTEXT(\"Input Latency Offset: {Val} ms\")", Val=@n5)
   )
  ]}]
variables []
widgets {
 info {inheritedWidgetCount=0, parentClass=&UserWidget, rootWidgetClass=&VerticalBox, widgetCount=3}
 widgets [
  VerticalBox_30 {widgetClassPath=&VerticalBox, widget=VerticalBox_30}
  AVOffset {
   widgetClassPath=&CommonTextBlock
   bIsVariable=true
   parent=VerticalBox_30
   widget=AVOffset
   properties {
    style=&TextStyle-Base_C
    text=A/V Offset: X
    font {
     bForceMonospaced=false
     bMaterialIsStencil=false
     fontMaterial=null
     fontObject=&NotoSans
     letterSpacing=0
     monospacedWidth=1
     size=24
     skewAmount=0
     typefaceFontName=NotoSans-Regular
     outlineSettings {
      bApplyOutlineToDropShadows=false
      bMiteredCorners=false
      bSeparateFillAlpha=false
      outlineMaterial=null
      outlineSize=0
      outlineColor=<rgba 0 0 0 1>
     }
    }
    shadowOffset=<0 0>
   }
   slot {class=VerticalBoxSlot}
  }
  InputOffset {
   widgetClassPath=&CommonTextBlock
   bIsVariable=true
   parent=VerticalBox_30
   widget=InputOffset
   properties {
    style=&TextStyle-Base_C
    text=Input Offset: X
    font {
     bForceMonospaced=false
     bMaterialIsStencil=false
     fontMaterial=null
     fontObject=&NotoSans
     letterSpacing=0
     monospacedWidth=1
     size=24
     skewAmount=0
     typefaceFontName=NotoSans-Regular
     outlineSettings {
      bApplyOutlineToDropShadows=false
      bMiteredCorners=false
      bSeparateFillAlpha=false
      outlineMaterial=null
      outlineSize=0
      outlineColor=<rgba 0 0 0 1>
     }
    }
    shadowOffset=<0 0>
   }
   slot {class=VerticalBoxSlot}
  }
 ]
}
